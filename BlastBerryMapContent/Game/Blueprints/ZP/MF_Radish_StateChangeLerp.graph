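# Material function that blends seven scalar state inputs, State_0 to State_6, into one value according to the Radish FX stage.
# It chains six lerps, each moving the running result toward the next state, so later stages override earlier ones.
# Each alpha comes from a SetStage parameter in MPC_Radish_FX, lerped toward a local SetState scalar parameter by Use_DebugStateChanges for debug control.
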MaterialFunction /BlastBerryMapContent/Game/Blueprints/ZP/MF_Radish_StateChangeLerp
graph {nodes [
  n1 = LinearInterpolate(
   A=FunctionInput() {inputName=State_0, inputType=FunctionInput_Scalar}
   B=FunctionInput() {inputName=State_1, inputType=FunctionInput_Scalar, sortPriority=1}
   Alpha=LinearInterpolate(
    A=CollectionParameter() {collection=/BlastBerryMapContent/Game/Blueprints/ZP/TechArt/Portals/MPC_Radish_FX, parameterName=SetStage_1}
    B=ScalarParameter() {parameterName=SetState_1}
    Alpha=ScalarParameter() {parameterName=Use_DebugStateChanges}
   )
  )
  n2 = FunctionInput() {inputName=State_2, inputType=FunctionInput_Scalar, sortPriority=2}
  n3 = LinearInterpolate(
   A=CollectionParameter() {collection=&MPC_Radish_FX, parameterName=SetStage_2}
   B=ScalarParameter() {parameterName=SetState_2}
   Alpha=ScalarParameter() {parameterName=Use_DebugStateChanges}
  )
  n4 = CollectionParameter() {collection=&MPC_Radish_FX, parameterName=SetStage_3}
  n5 = ScalarParameter() {parameterName=SetState_3}
  n6 = ScalarParameter() {parameterName=Use_DebugStateChanges}
  n7 = LinearInterpolate(
   A=LinearInterpolate(
    A=LinearInterpolate(A=@n1, B=@n2, Alpha=@n3)
    B=FunctionInput() {inputName=State_3, inputType=FunctionInput_Scalar, sortPriority=3}
    Alpha=LinearInterpolate(A=@n4, B=@n5, Alpha=@n6)
   )
   B=FunctionInput() {inputName=State_4, inputType=FunctionInput_Scalar, sortPriority=4}
   Alpha=LinearInterpolate(
    A=CollectionParameter() {collection=&MPC_Radish_FX, parameterName=SetStage_4}
    B=ScalarParameter() {parameterName=SetState_4}
    Alpha=ScalarParameter() {parameterName=Use_DebugStateChanges}
   )
  )
  n8 = FunctionInput() {inputName=State_5, inputType=FunctionInput_Scalar, sortPriority=5}
  n9 = LinearInterpolate(
   A=CollectionParameter() {collection=&MPC_Radish_FX, parameterName=SetStage_5}
   B=ScalarParameter() {parameterName=SetState_5}
   Alpha=ScalarParameter() {parameterName=Use_DebugStateChanges}
  )
  n10 = CollectionParameter() {collection=&MPC_Radish_FX, parameterName=SetStage_6}
  n11 = ScalarParameter() {parameterName=SetState_6}
  n12 = ScalarParameter() {parameterName=Use_DebugStateChanges}
  FunctionOutput(None=LinearInterpolate(
    A=LinearInterpolate(A=@n7, B=@n8, Alpha=@n9)
    B=FunctionInput() {inputName=State_6, inputType=FunctionInput_Scalar, sortPriority=6}
    Alpha=LinearInterpolate(A=@n10, B=@n11, Alpha=@n12)
   ))
 ]}
