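# Lit Zero Point material for the BlastBerry map, blending its look across seven states through MF_Radish_StateChangeLerp.
# Named reroutes feed state-blended parameters into reflection, dust, bump-offset deep noise, core flare and aura layers.
# Two UV sets blend between poles and middle, while object-stabilized radial screen UVs drive interior effects.
# MPC_Radish_FX controls added noise, tendril bursts, inner and outer glow, and flicker before the emissive adjustment.
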Material /BlastBerryMapContent/Game/Blueprints/ZP/Effects/ZeroPoint/StateChanges/M_Radish_ZeroPoint_MPC
graph {
 properties {bUsedWithNiagaraMeshParticles=true, pixelDepthOffsetMode=PDOM_Legacy, usedShadingModels=Default Lit}
 nodes [
  Constant()
  n1 = Multiply(A=TextureCoordinate() {coordinateIndex=1}, B=AppendVector(A=Constant() {r=1}, B=Constant() {r=1}))
  n2 = Multiply(
   A=Time()
   B=CollectionParameter() {collection=/BlastBerryMapContent/Game/Blueprints/ZP/TechArt/Portals/MPC_Radish_FX, parameterName=ExtraNoise_Pan_Scalar}
  )
  n3 = TextureSample(UVs=Rotator(Coordinate=@n1, Time=@n2) {speed=-1.1}) {texture=&AtmosphericCloudNoise01}
  n4 = Add(
   A=Rotator(Coordinate=@n1, Time=@n2) {speed=-1}
   B=Multiply(A=ScalarParameter() {defaultValue=0.02, parameterName=ExtraNoise_UV1_Distortion_Intensity}, B=@n3.R)
  )
  n5 = Multiply(A=OneMinus(None=@n3.R), B=ScalarParameter() {defaultValue=-2, parameterName=ExtraNoise_BumpOffsetAmount})
  n6 = Power(
   Base=TextureSample(UVs=BumpOffset(Coordinate=@n4, Height=@n5) {referencePlane=0}) {texture=&AtmosphericCloudNoise01}.RGB
   Exp=Constant() {r=3}
  )
  n7 = Multiply(A=TextureCoordinate() {coordinateIndex=1}, B=AppendVector(A=Constant() {r=1}, B=Constant() {r=1}))
  n8 = TextureSample(UVs=Rotator(Coordinate=@n7, Time=@n2) {speed=1.1}) {texture=&AtmosphericCloudNoise01}
  n9 = Add(
   A=Rotator(Coordinate=@n7, Time=@n2) {speed=1}
   B=Multiply(A=ScalarParameter() {defaultValue=0.02, parameterName=ExtraNoise_UV1_Distortion_Intensity}, B=@n8.R)
  )
  n10 = Multiply(A=OneMinus(None=@n8.R), B=ScalarParameter() {defaultValue=-2, parameterName=ExtraNoise_BumpOffsetAmount})
  n11 = Power(
   Base=TextureSample(UVs=BumpOffset(Coordinate=@n9, Height=@n10) {referencePlane=0}) {texture=&AtmosphericCloudNoise01}.RGB
   Exp=Constant() {r=2}
  )
  n12 = ComponentMask(None=TextureCoordinate()) {g=true}
  n13 = Clamp(None=Multiply(A=Power(Base=@n12) {constExponent=12}) {constB=100})
  n14 = CollectionParameter() {collection=&MPC_Radish_FX, parameterName=AddedNoise_PanX}
  n15 = Constant()
  n16 = Multiply(A=TextureCoordinate(), B=AppendVector(A=Constant() {r=1}, B=Constant() {r=1}))
  n17 = CollectionParameter() {collection=&MPC_Radish_FX, parameterName=ExtraNoise_Pan_Scalar}
  n18 = Multiply(A=ScalarParameter() {defaultValue=0.2, parameterName=ExtraNoise_Pan_X}, B=@n17)
  n19 = Multiply(A=@n17, B=ScalarParameter() {defaultValue=0.1, parameterName=ExtraNoise_Pan_Y})
  n20 = Multiply(A=ScalarParameter() {defaultValue=0.2, parameterName=ExtraNoise_Distortion_Pan_X}, B=@n17)
  n21 = Multiply(A=ScalarParameter() {defaultValue=0.05, parameterName=ExtraNoise_Distortion_Pan_Y}, B=@n17)
  n22 = TextureSample(UVs=Panner(Coordinate=@n16, Speed=AppendVector(A=@n20, B=@n21))) {texture=&AtmosphericCloudNoise01}
  n23 = Add(
   A=Panner(Coordinate=@n16, Speed=AppendVector(A=@n18, B=@n19))
   B=Multiply(A=ScalarParameter() {defaultValue=0.02, parameterName=ExtraNoise_UV1_Distortion_Intensity}, B=@n22.R)
  )
  n24 = Multiply(A=OneMinus(None=@n22.R), B=ScalarParameter() {defaultValue=-2, parameterName=ExtraNoise_BumpOffsetAmount})
  n25 = TextureSample(UVs=Add(A=AppendVector(A=@n14, B=@n15), B=BumpOffset(Coordinate=@n23, Height=@n24))) {texture=&AtmosphericCloudNoise01}
  n26 = Constant() {r=2}
  n27 = Sine(None=ComponentMask(None=TextureCoordinate() {vTiling=0.5}) {g=true})
  n28 = Constant() {r=3}
  n29 = Clamp(None=Multiply(A=Power(Base=@n27, Exp=@n28), B=Constant() {r=3}))
  n30 = Constant() {r=1}
  n31 = Multiply(
   A=Constant3Vector() {constant=<rgba 0 0.3261 1 1>}
   B=LinearInterpolate(
    A=LinearInterpolate(A=@n6, B=@n11, Alpha=@n13)
    B=Power(Base=@n25.RGB, Exp=@n26)
    Alpha=MaterialFunctionCall(In=@n29, Contrast=@n30) {materialFunction=&CheapContrast}.Result
   )
  )
  n32 = Constant() {r=1}
  n33 = Multiply(
   A=Power(
    Base=MaterialFunctionCall(Brightness=@n32, Exponent=@n32) {materialFunction=&MF_Fresnel_FX}.Fresnel
    Exp=CollectionParameter() {collection=&MPC_Radish_FX, parameterName=ExtraNoise_Fresnel_Power}
   )
   B=Constant() {r=0.2}
  )
  n34 = Multiply(A=TextureCoordinate() {coordinateIndex=1}, B=AppendVector(A=NamedRerouteUsage(), B=NamedRerouteUsage()))
  n35 = AppendVector(A=NamedRerouteUsage(), B=NamedRerouteUsage())
  n36 = Clamp(None=TextureSample(UVs=Panner(Coordinate=@n34, Speed=@n35)) {samplerType=SAMPLERTYPE_LinearGrayscale, texture=/Game/Global/Textures/Default/Noise/T_SwirlNoise_N_height}.R)
  n37 = NamedRerouteUsage()
  n38 = Multiply(A=TextureCoordinate(), B=AppendVector(A=NamedRerouteUsage(), B=NamedRerouteUsage()))
  n39 = AppendVector(A=NamedRerouteUsage(), B=NamedRerouteUsage())
  n40 = Clamp(None=TextureSample(UVs=Panner(Coordinate=@n38, Speed=@n39)) {samplerType=SAMPLERTYPE_LinearGrayscale, texture=&T_SwirlNoise_N_height}.R)
  n41 = ComponentMask(None=TextureCoordinate() {vTiling=0.5}) {g=true}
  n42 = Multiply(A=Power(Base=Sine(None=@n41), Exp=Constant() {r=3}), B=Constant() {r=3})
  n43 = LinearInterpolate(
   A=Constant()
   B=Constant() {r=1}
   Alpha=LinearInterpolate(A=Power(Base=@n36, Exp=@n37), B=Power(Base=@n40, Exp=@n37), Alpha=Clamp(None=@n42))
  )
  n44 = Constant() {r=1}
  n45 = Constant() {r=1}
  n46 = Multiply(
   A=Power(
    Base=MaterialFunctionCall(Brightness=@n44, Exponent=@n45) {materialFunction=&MF_Fresnel_FX}.Fresnel Inverted
    Exp=NamedRerouteUsage()
   )
   B=NamedRerouteUsage()
  )
  n47 = Clamp(None=LinearInterpolate(A=NamedRerouteUsage(), B=NamedRerouteUsage(), Alpha=Clamp(None=@n46)))
  n48 = Desaturation(
   None=TextureSample(UVs=ReflectionVectorWS()) {samplerType=SAMPLERTYPE_LinearColor, texture=/BlastBerryMapContent/Game/Blueprints/ZP/Effects/Tendrils/T_Radish_Skybox_FX}.RGB
   Fraction=NamedRerouteUsage()
  )
  n49 = Fresnel(ExponentIn=NamedRerouteUsage())
  n50 = Multiply(A=Power(Base=Multiply(A=@n48, B=@n49), Exp=NamedRerouteUsage()), B=NamedRerouteUsage())
  n51 = MaterialFunctionCall!(Position=ActorPositionWS(), UVRadialScale) {materialFunction=/Game/Weapons/WeaponSkins/MaterialFunction/ObjectStabilizedScreenSpaceUVs}
  n52 = Multiply(
   A=MaterialFunctionCall(Vector or UVs=Subtract(A=@n51.StabilizedUVs) {constB=0.5}) {materialFunction=&VectorToRadialValue}.Radial Coordinates
   B=AppendVector(A=NamedRerouteUsage(), B=NamedRerouteUsage())
  )
  n53 = AppendVector(A=NamedRerouteUsage(), B=NamedRerouteUsage())
  n54 = TextureCoordinate()
  n55 = LinearInterpolate(
   A=NamedRerouteUsage()
   B=NamedRerouteUsage()
   Alpha=TextureSample!(UVs=Panner(Coordinate=@n52, Speed=@n53), Tex, "DDX(UVs)"=DDX(Value=@n54), "DDY(UVs)"=DDY(Value=@n54), Apply View MipBias) {mipValueMode=TMVM_Derivative, texture=/Game/Global/Textures/Default/PointSplatter/T_Tiling_Dust}.R
  )
  n56 = Constant() {r=1}
  n57 = Power(
   Base=MaterialFunctionCall(Brightness=@n56, Exponent=@n56) {materialFunction=&MF_Fresnel_FX}.Fresnel
   Exp=NamedRerouteUsage()
  )
  n58 = NamedRerouteUsage()
  n59 = NamedRerouteUsage()
  n60 = NamedRerouteUsage()
  n61 = Multiply(A=TextureCoordinate() {coordinateIndex=1}, B=AppendVector(A=NamedRerouteUsage(), B=NamedRerouteUsage()))
  n62 = Multiply(A=TextureCoordinate() {coordinateIndex=1}, B=AppendVector(A=NamedRerouteUsage(), B=NamedRerouteUsage()))
  n63 = Multiply(
   A=OneMinus(None=TextureSample!(UVs=@n62, Tex, Apply View MipBias) {texture=&AtmosphericCloudNoise01}.R)
   B=NamedRerouteUsage()
  )
  n64 = OneMinus(None=TextureSample(UVs=BumpOffset(Coordinate=@n61, Height=@n63) {referencePlane=0}) {texture=&AtmosphericCloudNoise01}.R)
  n65 = Constant() {r=2.2}
  n66 = Multiply(A=TextureCoordinate(), B=AppendVector(A=NamedRerouteUsage(), B=NamedRerouteUsage()))
  n67 = Multiply(A=TextureCoordinate(), B=AppendVector(A=NamedRerouteUsage(), B=NamedRerouteUsage()))
  n68 = Multiply(A=OneMinus(None=TextureSample(UVs=@n67) {texture=&AtmosphericCloudNoise01}.R), B=NamedRerouteUsage())
  n69 = OneMinus(None=TextureSample(UVs=BumpOffset(Coordinate=@n66, Height=@n68) {referencePlane=0}) {texture=&AtmosphericCloudNoise01}.R)
  n70 = ComponentMask(None=TextureCoordinate() {vTiling=0.5}) {g=true}
  n71 = Multiply(A=Power(Base=Sine(None=@n70), Exp=Constant() {r=3}), B=Constant() {r=3})
  n72 = Power(Base=LinearInterpolate(A=Power(Base=@n64, Exp=@n65), B=Power(Base=@n69, Exp=@n65), Alpha=Clamp(None=@n71)), Exp=Constant() {r=2.2})
  n73 = Clamp(None=Multiply(A=Multiply(A=@n57, B=@n58), B=LinearInterpolate(A=@n59, B=@n60, Alpha=@n72)))
  n74 = Constant()
  n75 = Constant() {r=1}
  n76 = Constant() {r=1}
  n77 = NamedRerouteUsage()
  n78 = PixelNormalWS()
  n79 = Constant() {r=1}
  n80 = NamedRerouteUsage()
  n81 = ActorPositionWS()
  n82 = MaterialFunctionCall(Vector or UVs=Subtract(A=MaterialFunctionCall(Position=@n81) {materialFunction=&ObjectStabilizedScreenSpaceUVs}.StabilizedUVs) {constB=0.5}) {materialFunction=&VectorToRadialValue}
  n83 = AppendVector(A=NamedRerouteUsage(), B=NamedRerouteUsage())
  n84 = NamedRerouteUsage()
  n85 = NamedRerouteUsage()
  n86 = TextureCoordinate()
  n87 = TextureSample!(
   UVs=Panner(Coordinate=Multiply(A=@n82.Radial Coordinates, B=@n83), Speed=AppendVector(A=@n84, B=@n85))
   Tex
   "DDX(UVs)"=DDX(Value=@n86)
   "DDY(UVs)"=DDY(Value=@n86)
   Apply View MipBias
  ) {mipValueMode=TMVM_Derivative, texture=&AtmosphericCloudNoise01}
  n88 = TextureSample!(
   UVs=Power(
    Base=MaterialFunctionCall(Brightness=@n76, Exponent=@n77, Vertex / Pixel Normal=@n78) {materialFunction=&MF_Fresnel_FX}.Fresnel
    Exp=LinearInterpolate(A=@n79, B=@n80, Alpha=@n87.R)
   )
   Tex
   Apply View MipBias
  ) {
   samplerSource=SSM_Clamp_WorldGroupSettings
   samplerType=SAMPLERTYPE_LinearGrayscale
   texture=/Game/Characters/Player/Male/Medium/Bodies/M_MED_Cereal_Box/Textures/T_VertGradientLinear
  }
  n89 = Power(Base=@n88.R, Exp=NamedRerouteUsage())
  n90 = LinearInterpolate(A=Constant(), B=Constant() {r=0.5}, Alpha=Power(Base=@n88.R, Exp=NamedRerouteUsage()))
  n91 = Constant3Vector() {constant=<rgba 0 0.6521 2 1>}
  n92 = Add(
   A=Multiply(A=LinearInterpolate(A=@n74, B=@n75, Alpha=@n89), B=NamedRerouteUsage())
   B=Multiply(A=Multiply(A=@n90, B=@n91), B=NamedRerouteUsage())
  )
  n93 = MaterialFunctionCall(Position=ActorPositionWS()) {materialFunction=&ObjectStabilizedScreenSpaceUVs}
  n94 = Multiply(
   A=MaterialFunctionCall(Vector or UVs=Subtract(A=@n93.StabilizedUVs) {constB=0.5}) {materialFunction=&VectorToRadialValue}.Radial Coordinates
   B=AppendVector(A=NamedRerouteUsage(), B=NamedRerouteUsage())
  )
  n95 = ComponentMask(None=@n94) {g=true}
  n96 = Constant() {r=1}
  n97 = PixelNormalWS()
  n98 = NamedRerouteUsage()
  n99 = NamedRerouteUsage()
  n100 = ActorPositionWS()
  n101 = MaterialFunctionCall(Vector or UVs=Subtract(A=MaterialFunctionCall(Position=@n100) {materialFunction=&ObjectStabilizedScreenSpaceUVs}.StabilizedUVs) {constB=0.5}) {materialFunction=&VectorToRadialValue}
  n102 = AppendVector(A=NamedRerouteUsage(), B=NamedRerouteUsage())
  n103 = NamedRerouteUsage()
  n104 = NamedRerouteUsage()
  n105 = TextureCoordinate()
  n106 = TextureSample!(
   UVs=Panner(Coordinate=Multiply(A=@n101.Radial Coordinates, B=@n102), Speed=AppendVector(A=@n103, B=@n104))
   Tex
   "DDX(UVs)"=DDX(Value=@n105)
   "DDY(UVs)"=DDY(Value=@n105)
   Apply View MipBias
  ) {mipValueMode=TMVM_Derivative, texture=&AtmosphericCloudNoise01}
  n107 = TextureSample!(
   UVs=Power(
    Base=MaterialFunctionCall(Brightness=@n96, Exponent=@n96, Vertex / Pixel Normal=@n97) {materialFunction=&MF_Fresnel_FX}.Fresnel
    Exp=LinearInterpolate(A=@n98, B=@n99, Alpha=@n106.R)
   )
   Tex
   Apply View MipBias
  ) {
   samplerSource=SSM_Clamp_WorldGroupSettings
   samplerType=SAMPLERTYPE_LinearGrayscale
   texture=&T_VertGradientLinear
  }
  n108 = Multiply(A=@n107.R) {constB=2}
  n109 = NamedRerouteUsage()
  n110 = Multiply(A=Power(Base=Subtract(A=@n108), Exp=@n109), B=NamedRerouteUsage())
  n111 = Add(A=AppendVector(A=ComponentMask(None=@n94) {r=true}, B=Subtract(A=@n95, B=@n110)), B=AppendVector(A=Constant(), B=NamedRerouteUsage()))
  n112 = LinearInterpolate(A=Constant(), B=Constant() {r=1}, Alpha=Power(Base=@n107.R, Exp=@n109))
  n113 = LinearInterpolate(A=Constant(), B=NamedRerouteUsage(), Alpha=Power(Base=@n107.R, Exp=NamedRerouteUsage()))
  n114 = Multiply(
   A=Multiply(
    A=TextureSample(UVs=@n111) {samplerSource=SSM_Clamp_WorldGroupSettings, texture=/BlastBerryMapContent/Game/Blueprints/ZP/Tendrils/T_Radish_ZP_InnerGrad}.RGB
    B=NamedRerouteUsage()
   )
   B=Clamp(None=Add(A=@n112, B=@n113))
  )
  n115 = NamedRerouteUsage()
  n116 = NamedRerouteUsage()
  n117 = TextureSample(UVs=Multiply(A=TextureCoordinate() {coordinateIndex=1}, B=AppendVector(A=@n115, B=@n116))) {texture=/Game/Weapons/WeaponSkins/Textures/ScratchMap}
  n118 = Constant() {r=0.1}
  n119 = NamedRerouteUsage()
  n120 = NamedRerouteUsage()
  n121 = TextureSample(UVs=Multiply(A=TextureCoordinate(), B=AppendVector(A=@n119, B=@n120))) {texture=&ScratchMap}
  n122 = Constant() {r=0.1}
  n123 = Sine(None=ComponentMask(None=TextureCoordinate() {vTiling=0.5}) {g=true})
  n124 = Constant() {r=3}
  n125 = Clamp(None=Multiply(A=Power(Base=@n123, Exp=@n124), B=Constant() {r=3}))
  n126 = Constant() {r=1}
  n127 = Multiply(
   A=Multiply(A=Multiply(A=@n55, B=@n73) {constA=1}, B=Add(A=@n92, B=@n114))
   B=LinearInterpolate(
    A=Power(Base=@n117.RGB, Exp=@n118)
    B=Power(Base=@n121.RGB, Exp=@n122)
    Alpha=MaterialFunctionCall(In=@n125, Contrast=@n126) {materialFunction=&CheapContrast}.Result
   )
  )
  n128 = LinearInterpolate(A=@n50, B=Multiply(A=@n50, B=@n127), Alpha=Constant() {r=0.5})
  n129 = VertexNormalWS()
  n130 = Constant3Vector() {constant=<rgba 0 0 1 0>}
  n131 = Add(
   A=Multiply(A=Constant3Vector() {constant=<rgba 0 16.3027 50 1>}, B=Multiply(A=@n43, B=@n47))
   B=LinearInterpolate(A=Add(A=@n128, B=@n127), B=@n127, Alpha=DotProduct(A=@n129, B=@n130))
  )
  n132 = LinearInterpolate(
   A=@n131
   B=Add(A=@n131, B=Multiply(A=@n31, B=@n33))
   Alpha=Add(
    A=CollectionParameter() {collection=&MPC_Radish_FX, parameterName=Use_AddedNoise}
    B=CollectionParameter() {collection=&MPC_Radish_FX, parameterName=AddedNoise_TendrilBurst}
   )
  )
  n133 = Constant3Vector() {constant=<rgba 0 0.0326 0.1 1>}
  n134 = Constant() {r=1}
  n135 = Multiply(
   A=Power(
    Base=MaterialFunctionCall(Brightness=@n134, Exponent=@n134) {materialFunction=&MF_Fresnel_FX}.Fresnel
    Exp=ScalarParameter() {defaultValue=10, parameterName=Transition_Fresnel_Power}
   )
   B=ScalarParameter() {defaultValue=1, parameterName=Transition_Fresnel_Mult}
  )
  n136 = CollectionParameter() {collection=&MPC_Radish_FX, parameterName=ZP_InnerGlow}
  n137 = CollectionParameter() {collection=&MPC_Radish_FX, parameterName=InnerGlow_TendrilBurst}
  n138 = Constant()
  n139 = NamedRerouteUsage()
  n140 = ConstantBiasScale(None=Sine(None=Time()) {period=3})
  n141 = LinearInterpolate(
   A=@n132
   B=Add(A=Multiply(A=@n133, B=@n135), B=@n132)
   Alpha=Add(A=Add(A=@n136, B=@n137), B=LinearInterpolate(A=@n138, B=@n139, Alpha=@n140))
  )
  n142 = Constant3Vector() {constant=<rgba 0 0.3261 1 1>}
  n143 = Multiply(A=TextureCoordinate(), B=AppendVector(A=Constant() {r=1}, B=Constant() {r=1}))
  n144 = AppendVector(A=Constant() {r=0.01}, B=Constant() {r=0.2})
  n145 = Power(Base=TextureSample(UVs=Panner(Coordinate=@n143, Speed=@n144)) {samplerType=SAMPLERTYPE_LinearGrayscale, texture=&T_SwirlNoise_N_height}.R)
  n146 = Constant() {r=1}
  n147 = Power(Base=MaterialFunctionCall(Brightness=@n146, Exponent=@n146) {materialFunction=&MF_Fresnel_FX}.Fresnel Inverted) {constExponent=10}
  n148 = Multiply(A=Clamp(None=Multiply(A=@n145) {constB=2}), B=Clamp(None=Multiply(A=@n147) {constB=2}))
  n149 = LinearInterpolate(
   A=@n141
   B=Add(A=@n141, B=Multiply(A=@n142, B=@n148))
   Alpha=CollectionParameter() {collection=&MPC_Radish_FX, parameterName=ZP_OuterGlow}
  )
  n150 = Multiply(A=Sine(None=Time()) {period=0.2}, B=Sine(None=Time()) {period=0.3})
  n151 = LinearInterpolate(
   A=ScalarParameter() {parameterName=Flicker_Low}
   B=ScalarParameter() {defaultValue=5, parameterName=Flicker_High}
   Alpha=ConstantBiasScale(None=Clamp(None=@n150))
  )
  n152 = CollectionParameter() {collection=&MPC_Radish_FX, parameterName=ZP_Flicker}
  MaterialFunctionCall(Emissive=LinearInterpolate(A=@n149, B=Multiply(A=@n149, B=@n151), Alpha=Clamp(None=@n152))) {materialFunction=&MF_FX_EmissiveAdjustments}
  n153 = VertexNormalWS()
  n154 = TextureCoordinate() {coordinateIndex=1}
  n155 = AppendVector(A=NamedRerouteUsage(), B=NamedRerouteUsage())
  n156 = TextureCoordinate()
  n157 = AppendVector(A=NamedRerouteUsage(), B=NamedRerouteUsage())
  n158 = ComponentMask(None=TextureCoordinate() {vTiling=0.5}) {g=true}
  n159 = Multiply(A=Power(Base=Sine(None=@n158), Exp=Constant() {r=3}), B=Constant() {r=3})
  n160 = LinearInterpolate(
   A=TextureSample(UVs=Multiply(A=@n154, B=@n155)) {samplerType=SAMPLERTYPE_Normal, texture=/BlastBerryMapContent/Game/Blueprints/ZP/Effects/Tendrils/T_Radish_Scratches_N}.RGB
   B=TextureSample(UVs=Multiply(A=@n156, B=@n157)) {samplerType=SAMPLERTYPE_Normal, texture=&T_Radish_Scratches_N}.RGB
   Alpha=MaterialFunctionCall(In=Clamp(None=@n159), Contrast=Constant() {r=1}) {materialFunction=&CheapContrast}.Result
  )
  n161 = TextureSample() {samplerType=SAMPLERTYPE_Normal, texture=/Engine/EngineMaterials/FlatNormal}
  n162 = NamedRerouteUsage()
  NamedRerouteDeclaration(None=MaterialFunctionCall(
    BaseNormal=Transform(None=@n153) {transformSourceType=TRANSFORMSOURCE_World, transformType=TRANSFORM_Tangent}
    AdditionalNormal=LinearInterpolate(A=@n160, B=@n161.RGB, Alpha=@n162)
   ) {materialFunction=&BlendAngleCorrectedNormals}.Result) {name=Normal, nodeColor=<rgba 0.8281 1 0 1>}
  NamedRerouteDeclaration(None=MaterialFunctionCall!(
    State_0=ScalarParameter() {defaultValue=1, group=Dust Noise, parameterName=S0_Dust_Tile_X}
    State_1=ScalarParameter() {defaultValue=1, group=Dust Noise, parameterName=S1_Dust_Tile_X}
    State_2=ScalarParameter() {defaultValue=1, group=Dust Noise, parameterName=S2_Dust_Tile_X}
    State_3=ScalarParameter() {defaultValue=1, group=Dust Noise, parameterName=S3_Dust_Tile_X}
    State_4=ScalarParameter() {defaultValue=1, group=Dust Noise, parameterName=S4_Dust_Tile_X}
    State_5=ScalarParameter() {defaultValue=1, group=Dust Noise, parameterName=S5_Dust_Tile_X}
    State_6=ScalarParameter() {defaultValue=1, group=Dust Noise, parameterName=S6_Dust_Tile_X}
   ) {materialFunction=/BlastBerryMapContent/Game/Blueprints/ZP/MF_Radish_StateChangeLerp}.Result) {name=Dust_Tile_X, nodeColor=<rgba 1 0.3958 0 1>}
  NamedRerouteDeclaration(None=MaterialFunctionCall(
    State_0=ScalarParameter() {defaultValue=0.5, group=Reflection, parameterName=S0_Reflection_Desaturate}
    State_1=ScalarParameter() {defaultValue=0.5, group=Reflection, parameterName=S1_Reflection_Desaturate}
    State_2=ScalarParameter() {defaultValue=0.5, group=Reflection, parameterName=S2_Reflection_Desaturate}
    State_3=ScalarParameter() {defaultValue=0.5, group=Reflection, parameterName=S3_Reflection_Desaturate}
    State_4=ScalarParameter() {defaultValue=0.5, group=Reflection, parameterName=S4_Reflection_Desaturate}
    State_5=ScalarParameter() {defaultValue=0.5, group=Reflection, parameterName=S5_Reflection_Desaturate}
    State_6=ScalarParameter() {defaultValue=0.5, group=Reflection, parameterName=S6_Reflection_Desaturate}
   ) {materialFunction=&MF_Radish_StateChangeLerp}.Result) {name=Reflection_Desaturate, nodeColor=<rgba 1 0.5445 0 1>}
  NamedRerouteDeclaration(None=MaterialFunctionCall(
    State_0=ScalarParameter() {defaultValue=1, group=Deep Noise, parameterName=S0_Deep_Poles_Tile_Y}
    State_1=ScalarParameter() {defaultValue=1, group=Deep Noise, parameterName=S1_Deep_Poles_Tile_Y}
    State_2=ScalarParameter() {defaultValue=1, group=Deep Noise, parameterName=S2_Deep_Poles_Tile_Y}
    State_3=ScalarParameter() {defaultValue=1, group=Deep Noise, parameterName=S3_Deep_Poles_Tile_Y}
    State_4=ScalarParameter() {defaultValue=1, group=Deep Noise, parameterName=S4_Deep_Poles_Tile_Y}
    State_5=ScalarParameter() {defaultValue=1, group=Deep Noise, parameterName=S5_Deep_Poles_Tile_Y}
    State_6=ScalarParameter() {defaultValue=1, group=Deep Noise, parameterName=S6_Deep_Poles_Tile_Y}
   ) {materialFunction=&MF_Radish_StateChangeLerp}.Result) {name=Deep_Poles_Tile_Y, nodeColor=<rgba 1 0.4998 0 1>}
  NamedRerouteDeclaration(None=MaterialFunctionCall(
    State_0=ScalarParameter() {defaultValue=1, group=Reflection, parameterName=S0_Reflection_Fresnel_Exp}
    State_1=ScalarParameter() {defaultValue=1, group=Reflection, parameterName=S1_Reflection_Fresnel_Exp}
    State_2=ScalarParameter() {defaultValue=1, group=Reflection, parameterName=S2_Reflection_Fresnel_Exp}
    State_3=ScalarParameter() {defaultValue=1, group=Reflection, parameterName=S3_Reflection_Fresnel_Exp}
    State_4=ScalarParameter() {defaultValue=1, group=Reflection, parameterName=S4_Reflection_Fresnel_Exp}
    State_5=ScalarParameter() {defaultValue=1, group=Reflection, parameterName=S5_Reflection_Fresnel_Exp}
    State_6=ScalarParameter() {defaultValue=1, group=Reflection, parameterName=S6_Reflection_Fresnel_Exp}
   ) {materialFunction=&MF_Radish_StateChangeLerp}.Result) {name=Reflection_Fresnel_Exp, nodeColor=<rgba 1 0.5445 0 1>}
  NamedRerouteDeclaration(None=MaterialFunctionCall(
    State_0=ScalarParameter() {defaultValue=2, group=Deep Noise, parameterName=S0_Deep_Middle_Tile_X}
    State_1=ScalarParameter() {defaultValue=2, group=Deep Noise, parameterName=S1_Deep_Middle_Tile_X}
    State_2=ScalarParameter() {defaultValue=2, group=Deep Noise, parameterName=S2_Deep_Middle_Tile_X}
    State_3=ScalarParameter() {defaultValue=2, group=Deep Noise, parameterName=S3_Deep_Middle_Tile_X}
    State_4=ScalarParameter() {defaultValue=2, group=Deep Noise, parameterName=S4_Deep_Middle_Tile_X}
    State_5=ScalarParameter() {defaultValue=2, group=Deep Noise, parameterName=S5_Deep_Middle_Tile_X}
    State_6=ScalarParameter() {defaultValue=2, group=Deep Noise, parameterName=S6_Deep_Middle_Tile_X}
   ) {materialFunction=&MF_Radish_StateChangeLerp}.Result) {name=Deep_Middle_Tile_X, nodeColor=<rgba 1 0.4998 0 1>}
  NamedRerouteDeclaration(None=MaterialFunctionCall(
    State_0=ScalarParameter() {defaultValue=1, group=Deep Noise, parameterName=S0_Deep_Middle_Tile_Y}
    State_1=ScalarParameter() {defaultValue=1, group=Deep Noise, parameterName=S1_Deep_Middle_Tile_Y}
    State_2=ScalarParameter() {defaultValue=1, group=Deep Noise, parameterName=S2_Deep_Middle_Tile_Y}
    State_3=ScalarParameter() {defaultValue=1, group=Deep Noise, parameterName=S3_Deep_Middle_Tile_Y}
    State_4=ScalarParameter() {defaultValue=1, group=Deep Noise, parameterName=S4_Deep_Middle_Tile_Y}
    State_5=ScalarParameter() {defaultValue=1, group=Deep Noise, parameterName=S5_Deep_Middle_Tile_Y}
    State_6=ScalarParameter() {defaultValue=1, group=Deep Noise, parameterName=S6_Deep_Middle_Tile_Y}
   ) {materialFunction=&MF_Radish_StateChangeLerp}.Result) {name=Deep_Middle_Tile_Y, nodeColor=<rgba 1 0.4998 0 1>}
  NamedRerouteDeclaration(None=MaterialFunctionCall(
    State_0=ScalarParameter() {defaultValue=-2, group=Deep Noise, parameterName=S0_Deep_BumpOffset, sliderMin=-10}
    State_1=ScalarParameter() {defaultValue=-2, group=Deep Noise, parameterName=S1_Deep_BumpOffset, sliderMin=-10}
    State_2=ScalarParameter() {defaultValue=-2, group=Deep Noise, parameterName=S2_Deep_BumpOffset, sliderMin=-10}
    State_3=ScalarParameter() {defaultValue=-2, group=Deep Noise, parameterName=S3_Deep_BumpOffset, sliderMin=-10}
    State_4=ScalarParameter() {defaultValue=-2, group=Deep Noise, parameterName=S4_Deep_BumpOffset, sliderMin=-10}
    State_5=ScalarParameter() {defaultValue=-2, group=Deep Noise, parameterName=S5_Deep_BumpOffset, sliderMin=-10}
    State_6=ScalarParameter() {defaultValue=-2, group=Deep Noise, parameterName=S6_Deep_BumpOffset, sliderMin=-10}
   ) {materialFunction=&MF_Radish_StateChangeLerp}.Result) {name=Deep_BumpOffset, nodeColor=<rgba 1 0.4998 0 1>}
  NamedRerouteDeclaration(None=MaterialFunctionCall(
    State_0=ScalarParameter() {defaultValue=1, group=Deep Noise, parameterName=S0_Deep_Intensity_Max}
    State_1=ScalarParameter() {defaultValue=1, group=Deep Noise, parameterName=S1_Deep_Intensity_Max}
    State_2=ScalarParameter() {defaultValue=1, group=Deep Noise, parameterName=S2_Deep_Intensity_Max}
    State_3=ScalarParameter() {defaultValue=1, group=Deep Noise, parameterName=S3_Deep_Intensity_Max}
    State_4=ScalarParameter() {defaultValue=1, group=Deep Noise, parameterName=S4_Deep_Intensity_Max}
    State_5=ScalarParameter() {defaultValue=1, group=Deep Noise, parameterName=S5_Deep_Intensity_Max}
    State_6=ScalarParameter() {defaultValue=1, group=Deep Noise, parameterName=S6_Deep_Intensity_Max}
   ) {materialFunction=&MF_Radish_StateChangeLerp}.Result) {name=Deep_Intensity_Max, nodeColor=<rgba 1 0.4998 0 1>}
  NamedRerouteDeclaration(None=MaterialFunctionCall(
    State_0=ScalarParameter() {defaultValue=0.1, group=Deep Noise, parameterName=S0_Deep_Intensity_Min}
    State_1=ScalarParameter() {defaultValue=0.1, group=Deep Noise, parameterName=S1_Deep_Intensity_Min}
    State_2=ScalarParameter() {defaultValue=0.1, group=Deep Noise, parameterName=S2_Deep_Intensity_Min}
    State_3=ScalarParameter() {defaultValue=0.1, group=Deep Noise, parameterName=S3_Deep_Intensity_Min}
    State_4=ScalarParameter() {defaultValue=0.1, group=Deep Noise, parameterName=S4_Deep_Intensity_Min}
    State_5=ScalarParameter() {defaultValue=0.1, group=Deep Noise, parameterName=S5_Deep_Intensity_Min}
    State_6=ScalarParameter() {defaultValue=0.1, group=Deep Noise, parameterName=S6_Deep_Intensity_Min}
   ) {materialFunction=&MF_Radish_StateChangeLerp}.Result) {name=Deep_Intensity_Min, nodeColor=<rgba 1 0.4998 0 1>}
  NamedRerouteDeclaration(None=MaterialFunctionCall(
    State_0=ScalarParameter() {defaultValue=50, group=CoreFlare, parameterName=S0_CoreFlare_Fresnel_Exp}
    State_1=ScalarParameter() {defaultValue=50, group=CoreFlare, parameterName=S1_CoreFlare_Fresnel_Exp}
    State_2=ScalarParameter() {defaultValue=50, group=CoreFlare, parameterName=S2_CoreFlare_Fresnel_Exp}
    State_3=ScalarParameter() {defaultValue=50, group=CoreFlare, parameterName=S3_CoreFlare_Fresnel_Exp}
    State_4=ScalarParameter() {defaultValue=50, group=CoreFlare, parameterName=S4_CoreFlare_Fresnel_Exp}
    State_5=ScalarParameter() {defaultValue=50, group=CoreFlare, parameterName=S5_CoreFlare_Fresnel_Exp}
    State_6=ScalarParameter() {defaultValue=50, group=CoreFlare, parameterName=S6_CoreFlare_Fresnel_Exp}
   ) {materialFunction=&MF_Radish_StateChangeLerp}.Result) {name=CoreFlare_Fresnel_Exp, nodeColor=<rgba 1 0.3958 0 1>}
  NamedRerouteDeclaration(None=MaterialFunctionCall(
    State_0=ScalarParameter() {defaultValue=2, group=CoreFlare, parameterName=S0_CoreFlare_Distortion_Max}
    State_1=ScalarParameter() {defaultValue=2, group=CoreFlare, parameterName=S1_CoreFlare_Distortion_Max}
    State_2=ScalarParameter() {defaultValue=2, group=CoreFlare, parameterName=S2_CoreFlare_Distortion_Max}
    State_3=ScalarParameter() {defaultValue=2, group=CoreFlare, parameterName=S3_CoreFlare_Distortion_Max}
    State_4=ScalarParameter() {defaultValue=2, group=CoreFlare, parameterName=S4_CoreFlare_Distortion_Max}
    State_5=ScalarParameter() {defaultValue=2, group=CoreFlare, parameterName=S5_CoreFlare_Distortion_Max}
    State_6=ScalarParameter() {defaultValue=2, group=CoreFlare, parameterName=S6_CoreFlare_Distortion_Max}
   ) {materialFunction=&MF_Radish_StateChangeLerp}.Result) {name=CoreFlare_Distortion_Max, nodeColor=<rgba 1 0.3958 0 1>}
  NamedRerouteDeclaration(None=MaterialFunctionCall(
    State_0=ScalarParameter() {defaultValue=2, group=Dust Noise, parameterName=S0_Dust_Tile_Y}
    State_1=ScalarParameter() {defaultValue=2, group=Dust Noise, parameterName=S1_Dust_Tile_Y}
    State_2=ScalarParameter() {defaultValue=2, group=Dust Noise, parameterName=S2_Dust_Tile_Y}
    State_3=ScalarParameter() {defaultValue=2, group=Dust Noise, parameterName=S3_Dust_Tile_Y}
    State_4=ScalarParameter() {defaultValue=2, group=Dust Noise, parameterName=S4_Dust_Tile_Y}
    State_5=ScalarParameter() {defaultValue=2, group=Dust Noise, parameterName=S5_Dust_Tile_Y}
    State_6=ScalarParameter() {defaultValue=2, group=Dust Noise, parameterName=S6_Dust_Tile_Y}
   ) {materialFunction=&MF_Radish_StateChangeLerp}.Result) {name=Dust_Tile_Y, nodeColor=<rgba 1 0.3958 0 1>}
  NamedRerouteDeclaration(None=MaterialFunctionCall(
    State_0=ScalarParameter() {defaultValue=1, group=CoreFlare, parameterName=S0_CoreFlare_Tile_X}
    State_1=ScalarParameter() {defaultValue=1, group=CoreFlare, parameterName=S1_CoreFlare_Tile_X}
    State_2=ScalarParameter() {defaultValue=1, group=CoreFlare, parameterName=S2_CoreFlare_Tile_X}
    State_3=ScalarParameter() {defaultValue=1, group=CoreFlare, parameterName=S3_CoreFlare_Tile_X}
    State_4=ScalarParameter() {defaultValue=1, group=CoreFlare, parameterName=S4_CoreFlare_Tile_X}
    State_5=ScalarParameter() {defaultValue=1, group=CoreFlare, parameterName=S5_CoreFlare_Tile_X}
    State_6=ScalarParameter() {defaultValue=1, group=CoreFlare, parameterName=S6_CoreFlare_Tile_X}
   ) {materialFunction=&MF_Radish_StateChangeLerp}.Result) {name=CoreFlare_Tile_X, nodeColor=<rgba 1 0.3958 0 1>}
  NamedRerouteDeclaration(None=MaterialFunctionCall(
    State_0=ScalarParameter() {defaultValue=0.2, group=CoreFlare, parameterName=S0_CoreFlare_Tile_Y}
    State_1=ScalarParameter() {defaultValue=0.2, group=CoreFlare, parameterName=S1_CoreFlare_Tile_Y}
    State_2=ScalarParameter() {defaultValue=0.2, group=CoreFlare, parameterName=S2_CoreFlare_Tile_Y}
    State_3=ScalarParameter() {defaultValue=0.2, group=CoreFlare, parameterName=S3_CoreFlare_Tile_Y}
    State_4=ScalarParameter() {defaultValue=0.2, group=CoreFlare, parameterName=S4_CoreFlare_Tile_Y}
    State_5=ScalarParameter() {defaultValue=0.2, group=CoreFlare, parameterName=S5_CoreFlare_Tile_Y}
    State_6=ScalarParameter() {defaultValue=0.2, group=CoreFlare, parameterName=S6_CoreFlare_Tile_Y}
   ) {materialFunction=&MF_Radish_StateChangeLerp}.Result) {name=CoreFlare_Tile_Y, nodeColor=<rgba 1 0.3958 0 1>}
  NamedRerouteDeclaration(None=MaterialFunctionCall(
    State_0=ScalarParameter() {defaultValue=0.005, group=CoreFlare, parameterName=S0_CoreFlare_Pan_X}
    State_1=ScalarParameter() {defaultValue=0.005, group=CoreFlare, parameterName=S1_CoreFlare_Pan_X}
    State_2=ScalarParameter() {defaultValue=0.005, group=CoreFlare, parameterName=S2_CoreFlare_Pan_X}
    State_3=ScalarParameter() {defaultValue=0.005, group=CoreFlare, parameterName=S3_CoreFlare_Pan_X}
    State_4=ScalarParameter() {defaultValue=0.005, group=CoreFlare, parameterName=S4_CoreFlare_Pan_X}
    State_5=ScalarParameter() {defaultValue=0.005, group=CoreFlare, parameterName=S5_CoreFlare_Pan_X}
    State_6=ScalarParameter() {defaultValue=0.005, group=CoreFlare, parameterName=S6_CoreFlare_Pan_X}
   ) {materialFunction=&MF_Radish_StateChangeLerp}.Result) {name=CoreFlare_Pan_X, nodeColor=<rgba 1 0.3958 0 1>}
  NamedRerouteDeclaration(None=MaterialFunctionCall(
    State_0=ScalarParameter() {defaultValue=-0.025, group=CoreFlare, parameterName=S0_CoreFlare_Pan_Y}
    State_1=ScalarParameter() {defaultValue=-0.025, group=CoreFlare, parameterName=S1_CoreFlare_Pan_Y}
    State_2=ScalarParameter() {defaultValue=-0.025, group=CoreFlare, parameterName=S2_CoreFlare_Pan_Y}
    State_3=ScalarParameter() {defaultValue=-0.025, group=CoreFlare, parameterName=S3_CoreFlare_Pan_Y}
    State_4=ScalarParameter() {defaultValue=-0.025, group=CoreFlare, parameterName=S4_CoreFlare_Pan_Y}
    State_5=ScalarParameter() {defaultValue=-0.025, group=CoreFlare, parameterName=S5_CoreFlare_Pan_Y}
    State_6=ScalarParameter() {defaultValue=-0.025, group=CoreFlare, parameterName=S6_CoreFlare_Pan_Y}
   ) {materialFunction=&MF_Radish_StateChangeLerp}.Result) {name=CoreFlare_Pan_Y, nodeColor=<rgba 1 0.3958 0 1>}
  NamedRerouteDeclaration(None=MaterialFunctionCall(
    State_0=ScalarParameter() {defaultValue=100, group=CoreFlare, parameterName=S0_CoreFlare_Inner_Power}
    State_1=ScalarParameter() {defaultValue=100, group=CoreFlare, parameterName=S1_CoreFlare_Inner_Power}
    State_2=ScalarParameter() {defaultValue=100, group=CoreFlare, parameterName=S2_CoreFlare_Inner_Power}
    State_3=ScalarParameter() {defaultValue=100, group=CoreFlare, parameterName=S3_CoreFlare_Inner_Power}
    State_4=ScalarParameter() {defaultValue=100, group=CoreFlare, parameterName=S4_CoreFlare_Inner_Power}
    State_5=ScalarParameter() {defaultValue=100, group=CoreFlare, parameterName=S5_CoreFlare_Inner_Power}
    State_6=ScalarParameter() {defaultValue=100, group=CoreFlare, parameterName=S6_CoreFlare_Inner_Power}
   ) {materialFunction=&MF_Radish_StateChangeLerp}.Result) {name=CoreFlare_Inner_Power, nodeColor=<rgba 1 0.3958 0 1>}
  NamedRerouteDeclaration(None=MaterialFunctionCall(
    State_0=ScalarParameter() {defaultValue=20, group=CoreFlare, parameterName=S0_CoreFlare_Inner_Mult}
    State_1=ScalarParameter() {defaultValue=20, group=CoreFlare, parameterName=S1_CoreFlare_Inner_Mult}
    State_2=ScalarParameter() {defaultValue=20, group=CoreFlare, parameterName=S2_CoreFlare_Inner_Mult}
    State_3=ScalarParameter() {defaultValue=20, group=CoreFlare, parameterName=S3_CoreFlare_Inner_Mult}
    State_4=ScalarParameter() {defaultValue=20, group=CoreFlare, parameterName=S4_CoreFlare_Inner_Mult}
    State_5=ScalarParameter() {defaultValue=20, group=CoreFlare, parameterName=S5_CoreFlare_Inner_Mult}
    State_6=ScalarParameter() {defaultValue=20, group=CoreFlare, parameterName=S6_CoreFlare_Inner_Mult}
   ) {materialFunction=&MF_Radish_StateChangeLerp}.Result) {name=CoreFlare_Inner_Mult, nodeColor=<rgba 1 0.3958 0 1>}
  NamedRerouteDeclaration(None=MaterialFunctionCall(
    State_0=ScalarParameter() {defaultValue=10, group=CoreFlare, parameterName=S0_CoreFlare_Outer_Power}
    State_1=ScalarParameter() {defaultValue=10, group=CoreFlare, parameterName=S1_CoreFlare_Outer_Power}
    State_2=ScalarParameter() {defaultValue=10, group=CoreFlare, parameterName=S2_CoreFlare_Outer_Power}
    State_3=ScalarParameter() {defaultValue=10, group=CoreFlare, parameterName=S3_CoreFlare_Outer_Power}
    State_4=ScalarParameter() {defaultValue=10, group=CoreFlare, parameterName=S4_CoreFlare_Outer_Power}
    State_5=ScalarParameter() {defaultValue=10, group=CoreFlare, parameterName=S5_CoreFlare_Outer_Power}
    State_6=ScalarParameter() {defaultValue=10, group=CoreFlare, parameterName=S6_CoreFlare_Outer_Power}
   ) {materialFunction=&MF_Radish_StateChangeLerp}.Result) {name=CoreFlare_Outer_Power, nodeColor=<rgba 1 0.3958 0 1>}
  NamedRerouteDeclaration(None=MaterialFunctionCall(
    State_0=ScalarParameter() {defaultValue=1, group=CoreFlare, parameterName=S0_CoreFlare_Outer_Mult}
    State_1=ScalarParameter() {defaultValue=1, group=CoreFlare, parameterName=S1_CoreFlare_Outer_Mult}
    State_2=ScalarParameter() {defaultValue=1, group=CoreFlare, parameterName=S2_CoreFlare_Outer_Mult}
    State_3=ScalarParameter() {defaultValue=1, group=CoreFlare, parameterName=S3_CoreFlare_Outer_Mult}
    State_4=ScalarParameter() {defaultValue=1, group=CoreFlare, parameterName=S4_CoreFlare_Outer_Mult}
    State_5=ScalarParameter() {defaultValue=1, group=CoreFlare, parameterName=S5_CoreFlare_Outer_Mult}
    State_6=ScalarParameter() {defaultValue=1, group=CoreFlare, parameterName=S6_CoreFlare_Outer_Mult}
   ) {materialFunction=&MF_Radish_StateChangeLerp}.Result) {name=CoreFlare_Outer_Mult, nodeColor=<rgba 1 0.3958 0 1>}
  NamedRerouteDeclaration(None=MaterialFunctionCall(
    State_0=ScalarParameter() {defaultValue=1.5, group=Reflection, parameterName=S0_Reflection_Power}
    State_1=ScalarParameter() {defaultValue=1.5, group=Reflection, parameterName=S1_Reflection_Power}
    State_2=ScalarParameter() {defaultValue=1.5, group=Reflection, parameterName=S2_Reflection_Power}
    State_3=ScalarParameter() {defaultValue=1.5, group=Reflection, parameterName=S3_Reflection_Power}
    State_4=ScalarParameter() {defaultValue=1.5, group=Reflection, parameterName=S4_Reflection_Power}
    State_5=ScalarParameter() {defaultValue=1.5, group=Reflection, parameterName=S5_Reflection_Power}
    State_6=ScalarParameter() {defaultValue=1.5, group=Reflection, parameterName=S6_Reflection_Power}
   ) {materialFunction=&MF_Radish_StateChangeLerp}.Result) {name=Reflection_Power, nodeColor=<rgba 1 0.5445 0 1>}
  NamedRerouteDeclaration(None=MaterialFunctionCall(
    State_0=ScalarParameter() {defaultValue=1, group=Aura, parameterName=S0_Aura_Color_Tile_X}
    State_1=ScalarParameter() {defaultValue=1, group=Aura, parameterName=S1_Aura_Color_Tile_X}
    State_2=ScalarParameter() {defaultValue=1, group=Aura, parameterName=S2_Aura_Color_Tile_X}
    State_3=ScalarParameter() {defaultValue=1, group=Aura, parameterName=S3_Aura_Color_Tile_X}
    State_4=ScalarParameter() {defaultValue=1, group=Aura, parameterName=S4_Aura_Color_Tile_X}
    State_5=ScalarParameter() {defaultValue=1, group=Aura, parameterName=S5_Aura_Color_Tile_X}
    State_6=ScalarParameter() {defaultValue=1, group=Aura, parameterName=S6_Aura_Color_Tile_X}
   ) {materialFunction=&MF_Radish_StateChangeLerp}.Result) {name=Aura_Color_Tile_X, nodeColor=<rgba 1 0.3958 0 1>}
  NamedRerouteDeclaration(None=MaterialFunctionCall(
    State_0=ScalarParameter() {defaultValue=0.01, group=Dust Noise, parameterName=S0_Dust_Pan_X}
    State_1=ScalarParameter() {defaultValue=0.01, group=Dust Noise, parameterName=S1_Dust_Pan_X}
    State_2=ScalarParameter() {defaultValue=0.01, group=Dust Noise, parameterName=S2_Dust_Pan_X}
    State_3=ScalarParameter() {defaultValue=0.01, group=Dust Noise, parameterName=S3_Dust_Pan_X}
    State_4=ScalarParameter() {defaultValue=0.01, group=Dust Noise, parameterName=S4_Dust_Pan_X}
    State_5=ScalarParameter() {defaultValue=0.01, group=Dust Noise, parameterName=S5_Dust_Pan_X}
    State_6=ScalarParameter() {defaultValue=0.01, group=Dust Noise, parameterName=S6_Dust_Pan_X}
   ) {materialFunction=&MF_Radish_StateChangeLerp}.Result) {name=Dust_Pan_X, nodeColor=<rgba 1 0.3958 0 1>}
  NamedRerouteDeclaration(None=MaterialFunctionCall(
    State_0=ScalarParameter() {defaultValue=1.5, group=Aura, parameterName=S0_Aura_Color_Tile_Y}
    State_1=ScalarParameter() {defaultValue=1.5, group=Aura, parameterName=S1_Aura_Color_Tile_Y}
    State_2=ScalarParameter() {defaultValue=1.5, group=Aura, parameterName=S2_Aura_Color_Tile_Y}
    State_3=ScalarParameter() {defaultValue=1.5, group=Aura, parameterName=S3_Aura_Color_Tile_Y}
    State_4=ScalarParameter() {defaultValue=1.5, group=Aura, parameterName=S4_Aura_Color_Tile_Y}
    State_5=ScalarParameter() {defaultValue=1.5, group=Aura, parameterName=S5_Aura_Color_Tile_Y}
    State_6=ScalarParameter() {defaultValue=1.5, group=Aura, parameterName=S6_Aura_Color_Tile_Y}
   ) {materialFunction=&MF_Radish_StateChangeLerp}.Result) {name=Aura_Color_Tile_Y, nodeColor=<rgba 1 0.3958 0 1>}
  NamedRerouteDeclaration(None=MaterialFunctionCall(
    State_0=ScalarParameter() {defaultValue=0.1, group=Aura, parameterName=S0_Aura_Color_Distortion_Intensity}
    State_1=ScalarParameter() {defaultValue=0.1, group=Aura, parameterName=S1_Aura_Color_Distortion_Intensity}
    State_2=ScalarParameter() {defaultValue=0.1, group=Aura, parameterName=S2_Aura_Color_Distortion_Intensity}
    State_3=ScalarParameter() {defaultValue=0.1, group=Aura, parameterName=S3_Aura_Color_Distortion_Intensity}
    State_4=ScalarParameter() {defaultValue=0.1, group=Aura, parameterName=S4_Aura_Color_Distortion_Intensity}
    State_5=ScalarParameter() {defaultValue=0.1, group=Aura, parameterName=S5_Aura_Color_Distortion_Intensity}
    State_6=ScalarParameter() {defaultValue=0.1, group=Aura, parameterName=S6_Aura_Color_Distortion_Intensity}
   ) {materialFunction=&MF_Radish_StateChangeLerp}.Result) {name=Aura_Color_Distortion_Intensity, nodeColor=<rgba 1 0.3958 0 1>}
  NamedRerouteDeclaration(None=MaterialFunctionCall(
    State_0=ScalarParameter() {defaultValue=0.05, group=Aura, parameterName=S0_Aura_Color_Distortion_Offset_Y}
    State_1=ScalarParameter() {defaultValue=0.05, group=Aura, parameterName=S1_Aura_Color_Distortion_Offset_Y}
    State_2=ScalarParameter() {defaultValue=0.05, group=Aura, parameterName=S2_Aura_Color_Distortion_Offset_Y}
    State_3=ScalarParameter() {defaultValue=0.05, group=Aura, parameterName=S3_Aura_Color_Distortion_Offset_Y}
    State_4=ScalarParameter() {defaultValue=0.05, group=Aura, parameterName=S4_Aura_Color_Distortion_Offset_Y}
    State_5=ScalarParameter() {defaultValue=0.05, group=Aura, parameterName=S5_Aura_Color_Distortion_Offset_Y}
    State_6=ScalarParameter() {defaultValue=0.05, group=Aura, parameterName=S6_Aura_Color_Distortion_Offset_Y}
   ) {materialFunction=&MF_Radish_StateChangeLerp}.Result) {name=Aura_Color_Distortion_Offset_Y, nodeColor=<rgba 1 0.3958 0 1>}
  NamedRerouteDeclaration(None=MaterialFunctionCall(
    State_0=ScalarParameter() {defaultValue=2, group=Aura, parameterName=S0_Aura_Color_Intensity}
    State_1=ScalarParameter() {defaultValue=2, group=Aura, parameterName=S1_Aura_Color_Intensity}
    State_2=ScalarParameter() {defaultValue=2, group=Aura, parameterName=S2_Aura_Color_Intensity}
    State_3=ScalarParameter() {defaultValue=2, group=Aura, parameterName=S3_Aura_Color_Intensity}
    State_4=ScalarParameter() {defaultValue=2, group=Aura, parameterName=S4_Aura_Color_Intensity}
    State_5=ScalarParameter() {defaultValue=2, group=Aura, parameterName=S5_Aura_Color_Intensity}
    State_6=ScalarParameter() {defaultValue=2, group=Aura, parameterName=S6_Aura_Color_Intensity}
   ) {materialFunction=&MF_Radish_StateChangeLerp}.Result) {name=Aura_Color_Intensity, nodeColor=<rgba 1 0.3958 0 1>}
  NamedRerouteDeclaration(None=MaterialFunctionCall(
    State_0=ScalarParameter() {defaultValue=1, group=Aura, parameterName=S0_Aura_Noise_FresnelPower_Min, sortPriority=-1}
    State_1=ScalarParameter() {defaultValue=1, group=Aura, parameterName=S1_Aura_Noise_FresnelPower_Min, sortPriority=-1}
    State_2=ScalarParameter() {defaultValue=1, group=Aura, parameterName=S2_Aura_Noise_FresnelPower_Min, sortPriority=-1}
    State_3=ScalarParameter() {defaultValue=1, group=Aura, parameterName=S3_Aura_Noise_FresnelPower_Min, sortPriority=-1}
    State_4=ScalarParameter() {defaultValue=1, group=Aura, parameterName=S4_Aura_Noise_FresnelPower_Min, sortPriority=-1}
    State_5=ScalarParameter() {defaultValue=1, group=Aura, parameterName=S5_Aura_Noise_FresnelPower_Min, sortPriority=-1}
    State_6=ScalarParameter() {defaultValue=1, group=Aura, parameterName=S6_Aura_Noise_FresnelPower_Min, sortPriority=-1}
   ) {materialFunction=&MF_Radish_StateChangeLerp}.Result) {name=Aura_Noise_FresnelPower_Min, nodeColor=<rgba 1 0.3958 0 1>}
  NamedRerouteDeclaration(None=MaterialFunctionCall(
    State_0=ScalarParameter() {defaultValue=10, group=Aura, parameterName=S0_Aura_Noise_FresnelPower_Max, sortPriority=-1}
    State_1=ScalarParameter() {defaultValue=10, group=Aura, parameterName=S1_Aura_Noise_FresnelPower_Max, sortPriority=-1}
    State_2=ScalarParameter() {defaultValue=10, group=Aura, parameterName=S2_Aura_Noise_FresnelPower_Max, sortPriority=-1}
    State_3=ScalarParameter() {defaultValue=10, group=Aura, parameterName=S3_Aura_Noise_FresnelPower_Max, sortPriority=-1}
    State_4=ScalarParameter() {defaultValue=10, group=Aura, parameterName=S4_Aura_Noise_FresnelPower_Max, sortPriority=-1}
    State_5=ScalarParameter() {defaultValue=10, group=Aura, parameterName=S5_Aura_Noise_FresnelPower_Max, sortPriority=-1}
    State_6=ScalarParameter() {defaultValue=10, group=Aura, parameterName=S6_Aura_Noise_FresnelPower_Max, sortPriority=-1}
   ) {materialFunction=&MF_Radish_StateChangeLerp}.Result) {name=Aura_Noise_FresnelPower_Max, nodeColor=<rgba 1 0.3958 0 1>}
  NamedRerouteDeclaration(None=MaterialFunctionCall(
    State_0=ScalarParameter() {defaultValue=1, group=Aura, parameterName=S0_Aura_Noise_Tile_X}
    State_1=ScalarParameter() {defaultValue=1, group=Aura, parameterName=S1_Aura_Noise_Tile_X}
    State_2=ScalarParameter() {defaultValue=1, group=Aura, parameterName=S2_Aura_Noise_Tile_X}
    State_3=ScalarParameter() {defaultValue=1, group=Aura, parameterName=S3_Aura_Noise_Tile_X}
    State_4=ScalarParameter() {defaultValue=1, group=Aura, parameterName=S4_Aura_Noise_Tile_X}
    State_5=ScalarParameter() {defaultValue=1, group=Aura, parameterName=S5_Aura_Noise_Tile_X}
    State_6=ScalarParameter() {defaultValue=1, group=Aura, parameterName=S6_Aura_Noise_Tile_X}
   ) {materialFunction=&MF_Radish_StateChangeLerp}.Result) {name=Aura_Noise_Tile_X, nodeColor=<rgba 1 0.3958 0 1>}
  NamedRerouteDeclaration(None=MaterialFunctionCall(
    State_0=ScalarParameter() {defaultValue=0.2, group=Aura, parameterName=S0_Aura_Noise_Tile_Y}
    State_1=ScalarParameter() {defaultValue=0.2, group=Aura, parameterName=S1_Aura_Noise_Tile_Y}
    State_2=ScalarParameter() {defaultValue=0.2, group=Aura, parameterName=S2_Aura_Noise_Tile_Y}
    State_3=ScalarParameter() {defaultValue=0.2, group=Aura, parameterName=S3_Aura_Noise_Tile_Y}
    State_4=ScalarParameter() {defaultValue=0.2, group=Aura, parameterName=S4_Aura_Noise_Tile_Y}
    State_5=ScalarParameter() {defaultValue=0.2, group=Aura, parameterName=S5_Aura_Noise_Tile_Y}
    State_6=ScalarParameter() {defaultValue=0.2, group=Aura, parameterName=S6_Aura_Noise_Tile_Y}
   ) {materialFunction=&MF_Radish_StateChangeLerp}.Result) {name=Aura_Noise_Tile_Y, nodeColor=<rgba 1 0.3958 0 1>}
  NamedRerouteDeclaration(None=MaterialFunctionCall(
    State_0=ScalarParameter() {defaultValue=0.005, group=Aura, parameterName=S0_Aura_Noise_Pan_X}
    State_1=ScalarParameter() {defaultValue=0.005, group=Aura, parameterName=S1_Aura_Noise_Pan_X}
    State_2=ScalarParameter() {defaultValue=0.005, group=Aura, parameterName=S2_Aura_Noise_Pan_X}
    State_3=ScalarParameter() {defaultValue=0.005, group=Aura, parameterName=S3_Aura_Noise_Pan_X}
    State_4=ScalarParameter() {defaultValue=0.005, group=Aura, parameterName=S4_Aura_Noise_Pan_X}
    State_5=ScalarParameter() {defaultValue=0.005, group=Aura, parameterName=S5_Aura_Noise_Pan_X}
    State_6=ScalarParameter() {defaultValue=0.005, group=Aura, parameterName=S6_Aura_Noise_Pan_X}
   ) {materialFunction=&MF_Radish_StateChangeLerp}.Result) {name=Aura_Noise_Pan_X, nodeColor=<rgba 1 0.3958 0 1>}
  NamedRerouteDeclaration(None=MaterialFunctionCall(
    State_0=ScalarParameter() {defaultValue=-0.025, group=Aura, parameterName=S0_Aura_Noise_Pan_Y}
    State_1=ScalarParameter() {defaultValue=-0.025, group=Aura, parameterName=S1_Aura_Noise_Pan_Y}
    State_2=ScalarParameter() {defaultValue=-0.025, group=Aura, parameterName=S2_Aura_Noise_Pan_Y}
    State_3=ScalarParameter() {defaultValue=-0.025, group=Aura, parameterName=S3_Aura_Noise_Pan_Y}
    State_4=ScalarParameter() {defaultValue=-0.025, group=Aura, parameterName=S4_Aura_Noise_Pan_Y}
    State_5=ScalarParameter() {defaultValue=-0.025, group=Aura, parameterName=S5_Aura_Noise_Pan_Y}
    State_6=ScalarParameter() {defaultValue=-0.025, group=Aura, parameterName=S6_Aura_Noise_Pan_Y}
   ) {materialFunction=&MF_Radish_StateChangeLerp}.Result) {name=Aura_Noise_Pan_Y, nodeColor=<rgba 1 0.3958 0 1>}
  NamedRerouteDeclaration(None=MaterialFunctionCall(
    State_0=ScalarParameter() {defaultValue=-0.2, group=Dust Noise, parameterName=S0_Dust_Pan_Y}
    State_1=ScalarParameter() {defaultValue=-0.2, group=Dust Noise, parameterName=S1_Dust_Pan_Y}
    State_2=ScalarParameter() {defaultValue=-0.2, group=Dust Noise, parameterName=S2_Dust_Pan_Y}
    State_3=ScalarParameter() {defaultValue=-0.2, group=Dust Noise, parameterName=S3_Dust_Pan_Y}
    State_4=ScalarParameter() {defaultValue=-0.2, group=Dust Noise, parameterName=S4_Dust_Pan_Y}
    State_5=ScalarParameter() {defaultValue=-0.2, group=Dust Noise, parameterName=S5_Dust_Pan_Y}
    State_6=ScalarParameter() {defaultValue=-0.2, group=Dust Noise, parameterName=S6_Dust_Pan_Y}
   ) {materialFunction=&MF_Radish_StateChangeLerp}.Result) {name=Dust_Pan_Y, nodeColor=<rgba 1 0.3958 0 1>}
  NamedRerouteDeclaration(None=MaterialFunctionCall(
    State_0=ScalarParameter() {defaultValue=30, group=Aura, parameterName="S0_Aura_Noise_Inner_Power (For Scaling)"}
    State_1=ScalarParameter() {defaultValue=30, group=Aura, parameterName="S1_Aura_Noise_Inner_Power (For Scaling)"}
    State_2=ScalarParameter() {defaultValue=30, group=Aura, parameterName="S2_Aura_Noise_Inner_Power (For Scaling)"}
    State_3=ScalarParameter() {defaultValue=30, group=Aura, parameterName="S3_Aura_Noise_Inner_Power (For Scaling)"}
    State_4=ScalarParameter() {defaultValue=30, group=Aura, parameterName="S4_Aura_Noise_Inner_Power (For Scaling)"}
    State_5=ScalarParameter() {defaultValue=30, group=Aura, parameterName="S5_Aura_Noise_Inner_Power (For Scaling)"}
    State_6=ScalarParameter() {defaultValue=30, group=Aura, parameterName="S6_Aura_Noise_Inner_Power (For Scaling)"}
   ) {materialFunction=&MF_Radish_StateChangeLerp}.Result) {name="Aura_Noise_Inner_Power (For Scaling)", nodeColor=<rgba 1 0.3958 0 1>}
  NamedRerouteDeclaration(None=MaterialFunctionCall(
    State_0=ScalarParameter() {defaultValue=0.075, group=Aura, parameterName=S0_Aura_Noise_Outer_Intensity}
    State_1=ScalarParameter() {defaultValue=0.075, group=Aura, parameterName=S1_Aura_Noise_Outer_Intensity}
    State_2=ScalarParameter() {defaultValue=0.075, group=Aura, parameterName=S2_Aura_Noise_Outer_Intensity}
    State_3=ScalarParameter() {defaultValue=0.075, group=Aura, parameterName=S3_Aura_Noise_Outer_Intensity}
    State_4=ScalarParameter() {defaultValue=0.075, group=Aura, parameterName=S4_Aura_Noise_Outer_Intensity}
    State_5=ScalarParameter() {defaultValue=0.075, group=Aura, parameterName=S5_Aura_Noise_Outer_Intensity}
    State_6=ScalarParameter() {defaultValue=0.075, group=Aura, parameterName=S6_Aura_Noise_Outer_Intensity}
   ) {materialFunction=&MF_Radish_StateChangeLerp}.Result) {name=Aura_Noise_Outer_Intensity, nodeColor=<rgba 1 0.3958 0 1>}
  NamedRerouteDeclaration(None=MaterialFunctionCall(
    State_0=ScalarParameter() {defaultValue=3, group=Aura, parameterName="S0_Aura_Noise_Outer_Power (For Scaling)"}
    State_1=ScalarParameter() {defaultValue=3, group=Aura, parameterName="S1_Aura_Noise_Outer_Power (For Scaling)"}
    State_2=ScalarParameter() {defaultValue=3, group=Aura, parameterName="S2_Aura_Noise_Outer_Power (For Scaling)"}
    State_3=ScalarParameter() {defaultValue=3, group=Aura, parameterName="S3_Aura_Noise_Outer_Power (For Scaling)"}
    State_4=ScalarParameter() {defaultValue=3, group=Aura, parameterName="S4_Aura_Noise_Outer_Power (For Scaling)"}
    State_5=ScalarParameter() {defaultValue=3, group=Aura, parameterName="S5_Aura_Noise_Outer_Power (For Scaling)"}
    State_6=ScalarParameter() {defaultValue=3, group=Aura, parameterName="S6_Aura_Noise_Outer_Power (For Scaling)"}
   ) {materialFunction=&MF_Radish_StateChangeLerp}.Result) {name="Aura_Noise_Outer_Power (For Scaling)", nodeColor=<rgba 1 0.3958 0 1>}
  NamedRerouteDeclaration(None=MaterialFunctionCall(
    State_0=ScalarParameter() {defaultValue=4, group=Scratches, parameterName=S0_Scratches_Poles_Tile_X}
    State_1=ScalarParameter() {defaultValue=4, group=Scratches, parameterName=S1_Scratches_Poles_Tile_X}
    State_2=ScalarParameter() {defaultValue=4, group=Scratches, parameterName=S2_Scratches_Poles_Tile_X}
    State_3=ScalarParameter() {defaultValue=4, group=Scratches, parameterName=S3_Scratches_Poles_Tile_X}
    State_4=ScalarParameter() {defaultValue=4, group=Scratches, parameterName=S4_Scratches_Poles_Tile_X}
    State_5=ScalarParameter() {defaultValue=4, group=Scratches, parameterName=S5_Scratches_Poles_Tile_X}
    State_6=ScalarParameter() {defaultValue=4, group=Scratches, parameterName=S6_Scratches_Poles_Tile_X}
   ) {materialFunction=&MF_Radish_StateChangeLerp}.Result) {name=Scratches_Poles_Tile_X, nodeColor=<rgba 1 0.3958 0 1>}
  NamedRerouteDeclaration(None=MaterialFunctionCall(
    State_0=ScalarParameter() {defaultValue=4, group=Scratches, parameterName=S0_Scratches_Poles_Tile_Y}
    State_1=ScalarParameter() {defaultValue=4, group=Scratches, parameterName=S1_Scratches_Poles_Tile_Y}
    State_2=ScalarParameter() {defaultValue=4, group=Scratches, parameterName=S2_Scratches_Poles_Tile_Y}
    State_3=ScalarParameter() {defaultValue=4, group=Scratches, parameterName=S3_Scratches_Poles_Tile_Y}
    State_4=ScalarParameter() {defaultValue=4, group=Scratches, parameterName=S4_Scratches_Poles_Tile_Y}
    State_5=ScalarParameter() {defaultValue=4, group=Scratches, parameterName=S5_Scratches_Poles_Tile_Y}
    State_6=ScalarParameter() {defaultValue=4, group=Scratches, parameterName=S6_Scratches_Poles_Tile_Y}
   ) {materialFunction=&MF_Radish_StateChangeLerp}.Result) {name=Scratches_Poles_Tile_Y, nodeColor=<rgba 1 0.3958 0 1>}
  NamedRerouteDeclaration(None=MaterialFunctionCall(
    State_0=ScalarParameter() {defaultValue=4, group=Scratches, parameterName=S0_Scratches_Middle_Tile_X}
    State_1=ScalarParameter() {defaultValue=4, group=Scratches, parameterName=S1_Scratches_Middle_Tile_X}
    State_2=ScalarParameter() {defaultValue=4, group=Scratches, parameterName=S2_Scratches_Middle_Tile_X}
    State_3=ScalarParameter() {defaultValue=4, group=Scratches, parameterName=S3_Scratches_Middle_Tile_X}
    State_4=ScalarParameter() {defaultValue=4, group=Scratches, parameterName=S4_Scratches_Middle_Tile_X}
    State_5=ScalarParameter() {defaultValue=4, group=Scratches, parameterName=S5_Scratches_Middle_Tile_X}
    State_6=ScalarParameter() {defaultValue=4, group=Scratches, parameterName=S6_Scratches_Middle_Tile_X}
   ) {materialFunction=&MF_Radish_StateChangeLerp}.Result) {name=Scratches_Middle_Tile_X, nodeColor=<rgba 1 0.3958 0 1>}
  NamedRerouteDeclaration(None=MaterialFunctionCall(
    State_0=ScalarParameter() {defaultValue=4, group=Scratches, parameterName=S0_Scratches_Middle_Tile_Y}
    State_1=ScalarParameter() {defaultValue=4, group=Scratches, parameterName=S1_Scratches_Middle_Tile_Y}
    State_2=ScalarParameter() {defaultValue=4, group=Scratches, parameterName=S2_Scratches_Middle_Tile_Y}
    State_3=ScalarParameter() {defaultValue=4, group=Scratches, parameterName=S3_Scratches_Middle_Tile_Y}
    State_4=ScalarParameter() {defaultValue=4, group=Scratches, parameterName=S4_Scratches_Middle_Tile_Y}
    State_5=ScalarParameter() {defaultValue=4, group=Scratches, parameterName=S5_Scratches_Middle_Tile_Y}
    State_6=ScalarParameter() {defaultValue=4, group=Scratches, parameterName=S6_Scratches_Middle_Tile_Y}
   ) {materialFunction=&MF_Radish_StateChangeLerp}.Result) {name=Scratches_Middle_Tile_Y, nodeColor=<rgba 1 0.3958 0 1>}
  n163 = LinearInterpolate(
   A=VectorParameter() {group=Reflection, parameterName="S0_Reflection_Tint&&Intensity", defaultValue=<rgba 0.25 0.3315 0.5 1>}.RGB
   B=VectorParameter() {group=Reflection, parameterName="S1_Reflection_Tint&&Intensity", defaultValue=<rgba 0.25 0.3315 0.5 1>}.RGB
   Alpha=ScalarParameter() {parameterName=SetState_1}
  )
  n164 = VectorParameter() {group=Reflection, parameterName="S2_Reflection_Tint&&Intensity", defaultValue=<rgba 0.25 0.3315 0.5 1>}
  n165 = ScalarParameter() {parameterName=SetState_2}
  n166 = LinearInterpolate(
   A=LinearInterpolate(
    A=LinearInterpolate(A=@n163, B=@n164.RGB, Alpha=@n165)
    B=VectorParameter() {group=Reflection, parameterName="S3_Reflection_Tint&&Intensity", defaultValue=<rgba 0.25 0.3315 0.5 1>}.RGB
    Alpha=ScalarParameter() {parameterName=SetState_3}
   )
   B=VectorParameter() {group=Reflection, parameterName="S4_Reflection_Tint&&Intensity", defaultValue=<rgba 0.25 0.3315 0.5 1>}.RGB
   Alpha=ScalarParameter() {parameterName=SetState_4}
  )
  n167 = VectorParameter() {group=Reflection, parameterName="S5_Reflection_Tint&&Intensity", defaultValue=<rgba 0.25 0.3315 0.5 1>}
  n168 = ScalarParameter() {parameterName=SetState_5}
  NamedRerouteDeclaration(None=LinearInterpolate(
    A=LinearInterpolate(A=@n166, B=@n167.RGB, Alpha=@n168)
    B=VectorParameter() {group=Reflection, parameterName="S6_Reflection_Tint&&Intensity", defaultValue=<rgba 0.25 0.3315 0.5 1>}.RGB
    Alpha=ScalarParameter() {parameterName=SetState_6}
   )) {name="Reflection_Tint&&Intensity", nodeColor=<rgba 1 0.5445 0 1>}
  NamedRerouteDeclaration(None=MaterialFunctionCall(
    State_0=ScalarParameter() {defaultValue=0.9, group=Scratches, parameterName=S0_Scratches_Normal_Flatten}
    State_1=ScalarParameter() {defaultValue=0.9, group=Scratches, parameterName=S1_Scratches_Normal_Flatten}
    State_2=ScalarParameter() {defaultValue=0.9, group=Scratches, parameterName=S2_Scratches_Normal_Flatten}
    State_3=ScalarParameter() {defaultValue=0.9, group=Scratches, parameterName=S3_Scratches_Normal_Flatten}
    State_4=ScalarParameter() {defaultValue=0.9, group=Scratches, parameterName=S4_Scratches_Normal_Flatten}
    State_5=ScalarParameter() {defaultValue=0.9, group=Scratches, parameterName=S5_Scratches_Normal_Flatten}
    State_6=ScalarParameter() {defaultValue=0.9, group=Scratches, parameterName=S6_Scratches_Normal_Flatten}
   ) {materialFunction=&MF_Radish_StateChangeLerp}.Result) {name=Scratches_Normal_Flatten, nodeColor=<rgba 1 0.3958 0 1>}
  NamedRerouteDeclaration(None=MaterialFunctionCall(
    State_0=ScalarParameter() {defaultValue=1, group=Rim, parameterName=RimNoise_Poles_Tile_X}
    State_1=ScalarParameter() {defaultValue=1, group=Rim, parameterName=RimNoise_Poles_Tile_X}
    State_2=ScalarParameter() {defaultValue=1, group=Rim, parameterName=RimNoise_Poles_Tile_X}
    State_3=ScalarParameter() {defaultValue=1, group=Rim, parameterName=RimNoise_Poles_Tile_X}
    State_4=ScalarParameter() {defaultValue=1, group=Rim, parameterName=RimNoise_Poles_Tile_X}
    State_5=ScalarParameter() {defaultValue=1, group=Rim, parameterName=RimNoise_Poles_Tile_X}
    State_6=ScalarParameter() {defaultValue=1, group=Rim, parameterName=RimNoise_Poles_Tile_X}
   ) {materialFunction=&MF_Radish_StateChangeLerp}.Result) {name=RimNoise_Poles_Tile_X, nodeColor=<rgba 1 0.3958 0 1>}
  NamedRerouteDeclaration(None=MaterialFunctionCall(
    State_0=ScalarParameter() {defaultValue=0.5, group=Dust Noise, parameterName=S0_Dust_Intensity_Min}
    State_1=ScalarParameter() {defaultValue=0.5, group=Dust Noise, parameterName=S1_Dust_Intensity_Min}
    State_2=ScalarParameter() {defaultValue=0.5, group=Dust Noise, parameterName=S2_Dust_Intensity_Min}
    State_3=ScalarParameter() {defaultValue=0.5, group=Dust Noise, parameterName=S3_Dust_Intensity_Min}
    State_4=ScalarParameter() {defaultValue=0.5, group=Dust Noise, parameterName=S4_Dust_Intensity_Min}
    State_5=ScalarParameter() {defaultValue=0.5, group=Dust Noise, parameterName=S5_Dust_Intensity_Min}
    State_6=ScalarParameter() {defaultValue=0.5, group=Dust Noise, parameterName=S6_Dust_Intensity_Min}
   ) {materialFunction=&MF_Radish_StateChangeLerp}.Result) {name=Dust_Intensity_Min, nodeColor=<rgba 1 0.3958 0 1>}
  NamedRerouteDeclaration(None=MaterialFunctionCall(
    State_0=ScalarParameter() {defaultValue=1, group=Rim, parameterName=S0_RimNoise_Poles_Tile_Y}
    State_1=ScalarParameter() {defaultValue=1, group=Rim, parameterName=S1_RimNoise_Poles_Tile_Y}
    State_2=ScalarParameter() {defaultValue=1, group=Rim, parameterName=S2_RimNoise_Poles_Tile_Y}
    State_3=ScalarParameter() {defaultValue=1, group=Rim, parameterName=S3_RimNoise_Poles_Tile_Y}
    State_4=ScalarParameter() {defaultValue=1, group=Rim, parameterName=S4_RimNoise_Poles_Tile_Y}
    State_5=ScalarParameter() {defaultValue=1, group=Rim, parameterName=S5_RimNoise_Poles_Tile_Y}
    State_6=ScalarParameter() {defaultValue=1, group=Rim, parameterName=S6_RimNoise_Poles_Tile_Y}
   ) {materialFunction=&MF_Radish_StateChangeLerp}.Result) {name=RimNoise_Poles_Tile_Y, nodeColor=<rgba 1 0.3958 0 1>}
  NamedRerouteDeclaration(None=MaterialFunctionCall(
    State_0=ScalarParameter() {defaultValue=0.01, group=Rim, parameterName=S0_RimNoise_Poles_2_Pan_X}
    State_1=ScalarParameter() {defaultValue=0.01, group=Rim, parameterName=S1_RimNoise_Poles_2_Pan_X}
    State_2=ScalarParameter() {defaultValue=0.01, group=Rim, parameterName=S2_RimNoise_Poles_2_Pan_X}
    State_3=ScalarParameter() {defaultValue=0.01, group=Rim, parameterName=S3_RimNoise_Poles_2_Pan_X}
    State_4=ScalarParameter() {defaultValue=0.01, group=Rim, parameterName=S4_RimNoise_Poles_2_Pan_X}
    State_5=ScalarParameter() {defaultValue=0.01, group=Rim, parameterName=S5_RimNoise_Poles_2_Pan_X}
    State_6=ScalarParameter() {defaultValue=0.01, group=Rim, parameterName=S6_RimNoise_Poles_2_Pan_X}
   ) {materialFunction=&MF_Radish_StateChangeLerp}.Result) {name=RimNoise_Poles_2_Pan_X, nodeColor=<rgba 1 0.3958 0 1>}
  NamedRerouteDeclaration(None=MaterialFunctionCall(
    State_0=ScalarParameter() {defaultValue=0.05, group=Rim, parameterName=S0_RimNoise_Poles_2_Pan_Y}
    State_1=ScalarParameter() {defaultValue=0.05, group=Rim, parameterName=S1_RimNoise_Poles_2_Pan_Y}
    State_2=ScalarParameter() {defaultValue=0.05, group=Rim, parameterName=S2_RimNoise_Poles_2_Pan_Y}
    State_3=ScalarParameter() {defaultValue=0.05, group=Rim, parameterName=S3_RimNoise_Poles_2_Pan_Y}
    State_4=ScalarParameter() {defaultValue=0.05, group=Rim, parameterName=S4_RimNoise_Poles_2_Pan_Y}
    State_5=ScalarParameter() {defaultValue=0.05, group=Rim, parameterName=S5_RimNoise_Poles_2_Pan_Y}
    State_6=ScalarParameter() {defaultValue=0.05, group=Rim, parameterName=S6_RimNoise_Poles_2_Pan_Y}
   ) {materialFunction=&MF_Radish_StateChangeLerp}.Result) {name=RimNoise_Poles_2_Pan_Y, nodeColor=<rgba 1 0.3958 0 1>}
  NamedRerouteDeclaration(None=MaterialFunctionCall(
    State_0=ScalarParameter() {defaultValue=2, group=Rim, parameterName=S0_RimNoise_Mid_Tile_X}
    State_1=ScalarParameter() {defaultValue=2, group=Rim, parameterName=S1_RimNoise_Mid_Tile_X}
    State_2=ScalarParameter() {defaultValue=2, group=Rim, parameterName=S2_RimNoise_Mid_Tile_X}
    State_3=ScalarParameter() {defaultValue=2, group=Rim, parameterName=S3_RimNoise_Mid_Tile_X}
    State_4=ScalarParameter() {defaultValue=2, group=Rim, parameterName=S4_RimNoise_Mid_Tile_X}
    State_5=ScalarParameter() {defaultValue=2, group=Rim, parameterName=S5_RimNoise_Mid_Tile_X}
    State_6=ScalarParameter() {defaultValue=2, group=Rim, parameterName=S6_RimNoise_Mid_Tile_X}
   ) {materialFunction=&MF_Radish_StateChangeLerp}.Result) {name=RimNoise_Mid_Tile_X, nodeColor=<rgba 1 0.3958 0 1>}
  NamedRerouteDeclaration(None=MaterialFunctionCall(
    State_0=ScalarParameter() {defaultValue=1, group=Rim, parameterName=S0_RimNoise_Mid_Tile_Y}
    State_1=ScalarParameter() {defaultValue=1, group=Rim, parameterName=S1_RimNoise_Mid_Tile_Y}
    State_2=ScalarParameter() {defaultValue=1, group=Rim, parameterName=S2_RimNoise_Mid_Tile_Y}
    State_3=ScalarParameter() {defaultValue=1, group=Rim, parameterName=S3_RimNoise_Mid_Tile_Y}
    State_4=ScalarParameter() {defaultValue=1, group=Rim, parameterName=S4_RimNoise_Mid_Tile_Y}
    State_5=ScalarParameter() {defaultValue=1, group=Rim, parameterName=S5_RimNoise_Mid_Tile_Y}
    State_6=ScalarParameter() {defaultValue=1, group=Rim, parameterName=S6_RimNoise_Mid_Tile_Y}
   ) {materialFunction=&MF_Radish_StateChangeLerp}.Result) {name=RimNoise_Mid_Tile_Y, nodeColor=<rgba 1 0.3958 0 1>}
  NamedRerouteDeclaration(None=MaterialFunctionCall(
    State_0=ScalarParameter() {defaultValue=0.05, group=Rim, parameterName=S0_RimNoise_Mid_Pan_X}
    State_1=ScalarParameter() {defaultValue=0.05, group=Rim, parameterName=S1_RimNoise_Mid_Pan_X}
    State_2=ScalarParameter() {defaultValue=0.05, group=Rim, parameterName=S2_RimNoise_Mid_Pan_X}
    State_3=ScalarParameter() {defaultValue=0.05, group=Rim, parameterName=S3_RimNoise_Mid_Pan_X}
    State_4=ScalarParameter() {defaultValue=0.05, group=Rim, parameterName=S4_RimNoise_Mid_Pan_X}
    State_5=ScalarParameter() {defaultValue=0.05, group=Rim, parameterName=S5_RimNoise_Mid_Pan_X}
    State_6=ScalarParameter() {defaultValue=0.05, group=Rim, parameterName=S6_RimNoise_Mid_Pan_X}
   ) {materialFunction=&MF_Radish_StateChangeLerp}.Result) {name=RimNoise_Mid_Pan_X, nodeColor=<rgba 1 0.3958 0 1>}
  NamedRerouteDeclaration(None=MaterialFunctionCall(
    State_0=ScalarParameter() {defaultValue=0.01, group=Rim, parameterName=S0_RimNoise_Mid_Pan_Y, sliderMin=0.025}
    State_1=ScalarParameter() {defaultValue=0.01, group=Rim, parameterName=S1_RimNoise_Mid_Pan_Y, sliderMin=0.025}
    State_2=ScalarParameter() {defaultValue=0.01, group=Rim, parameterName=S2_RimNoise_Mid_Pan_Y, sliderMin=0.025}
    State_3=ScalarParameter() {defaultValue=0.01, group=Rim, parameterName=S3_RimNoise_Mid_Pan_Y, sliderMin=0.025}
    State_4=ScalarParameter() {defaultValue=0.01, group=Rim, parameterName=S4_RimNoise_Mid_Pan_Y, sliderMin=0.025}
    State_5=ScalarParameter() {defaultValue=0.01, group=Rim, parameterName=S5_RimNoise_Mid_Pan_Y, sliderMin=0.025}
    State_6=ScalarParameter() {defaultValue=0.01, group=Rim, parameterName=S6_RimNoise_Mid_Pan_Y, sliderMin=0.025}
   ) {materialFunction=&MF_Radish_StateChangeLerp}.Result) {name=RimNoise_Mid_Pan_Y, nodeColor=<rgba 1 0.3958 0 1>}
  NamedRerouteDeclaration(None=MaterialFunctionCall(
    State_0=ScalarParameter() {defaultValue=10, group=Rim, parameterName=S0_RimNoise_Power}
    State_1=ScalarParameter() {defaultValue=10, group=Rim, parameterName=S1_RimNoise_Power}
    State_2=ScalarParameter() {defaultValue=10, group=Rim, parameterName=S2_RimNoise_Power}
    State_3=ScalarParameter() {defaultValue=10, group=Rim, parameterName=S3_RimNoise_Power}
    State_4=ScalarParameter() {defaultValue=10, group=Rim, parameterName=S4_RimNoise_Power}
    State_5=ScalarParameter() {defaultValue=10, group=Rim, parameterName=S5_RimNoise_Power}
    State_6=ScalarParameter() {defaultValue=10, group=Rim, parameterName=S6_RimNoise_Power}
   ) {materialFunction=&MF_Radish_StateChangeLerp}.Result) {name=RimNoise_Power, nodeColor=<rgba 1 0.3958 0 1>}
  NamedRerouteDeclaration(None=MaterialFunctionCall(
    State_0=ScalarParameter() {defaultValue=18, group=Rim, parameterName=S0_RimNoise_Fresnel_Power}
    State_1=ScalarParameter() {defaultValue=18, group=Rim, parameterName=S1_RimNoise_Fresnel_Power}
    State_2=ScalarParameter() {defaultValue=18, group=Rim, parameterName=S2_RimNoise_Fresnel_Power}
    State_3=ScalarParameter() {defaultValue=18, group=Rim, parameterName=S3_RimNoise_Fresnel_Power}
    State_4=ScalarParameter() {defaultValue=18, group=Rim, parameterName=S4_RimNoise_Fresnel_Power}
    State_5=ScalarParameter() {defaultValue=18, group=Rim, parameterName=S5_RimNoise_Fresnel_Power}
    State_6=ScalarParameter() {defaultValue=18, group=Rim, parameterName=S6_RimNoise_Fresnel_Power}
   ) {materialFunction=&MF_Radish_StateChangeLerp}.Result) {name=RimNoise_Fresnel_Power, nodeColor=<rgba 1 0.3958 0 1>}
  NamedRerouteDeclaration(None=MaterialFunctionCall(
    State_0=ScalarParameter() {defaultValue=1, group=Rim, parameterName=S0_RimNoise_Fresnel_Mult}
    State_1=ScalarParameter() {defaultValue=1, group=Rim, parameterName=S1_RimNoise_Fresnel_Mult}
    State_2=ScalarParameter() {defaultValue=1, group=Rim, parameterName=S2_RimNoise_Fresnel_Mult}
    State_3=ScalarParameter() {defaultValue=1, group=Rim, parameterName=S3_RimNoise_Fresnel_Mult}
    State_4=ScalarParameter() {defaultValue=1, group=Rim, parameterName=S4_RimNoise_Fresnel_Mult}
    State_5=ScalarParameter() {defaultValue=1, group=Rim, parameterName=S5_RimNoise_Fresnel_Mult}
    State_6=ScalarParameter() {defaultValue=1, group=Rim, parameterName=S6_RimNoise_Fresnel_Mult}
   ) {materialFunction=&MF_Radish_StateChangeLerp}.Result) {name=RimNoise_Fresnel_Mult, nodeColor=<rgba 1 0.3958 0 1>}
  NamedRerouteDeclaration(None=MaterialFunctionCall(
    State_0=ScalarParameter() {defaultValue=2, group=Dust Noise, parameterName=S0_Dust_Intensity_Max}
    State_1=ScalarParameter() {defaultValue=2, group=Dust Noise, parameterName=S1_Dust_Intensity_Max}
    State_2=ScalarParameter() {defaultValue=2, group=Dust Noise, parameterName=S2_Dust_Intensity_Max}
    State_3=ScalarParameter() {defaultValue=2, group=Dust Noise, parameterName=S3_Dust_Intensity_Max}
    State_4=ScalarParameter() {defaultValue=2, group=Dust Noise, parameterName=S4_Dust_Intensity_Max}
    State_5=ScalarParameter() {defaultValue=2, group=Dust Noise, parameterName=S5_Dust_Intensity_Max}
    State_6=ScalarParameter() {defaultValue=2, group=Dust Noise, parameterName=S6_Dust_Intensity_Max}
   ) {materialFunction=&MF_Radish_StateChangeLerp}.Result) {name=Dust_Intensity_Max, nodeColor=<rgba 1 0.3958 0 1>}
  NamedRerouteDeclaration(None=MaterialFunctionCall(
    State_0=ScalarParameter() {group=Rim, parameterName=S0_RimNoise_Fresnel_Min}
    State_1=ScalarParameter() {group=Rim, parameterName=S1_RimNoise_Fresnel_Min}
    State_2=ScalarParameter() {group=Rim, parameterName=S2_RimNoise_Fresnel_Min}
    State_3=ScalarParameter() {group=Rim, parameterName=S3_RimNoise_Fresnel_Min}
    State_4=ScalarParameter() {group=Rim, parameterName=S4_RimNoise_Fresnel_Min}
    State_5=ScalarParameter() {group=Rim, parameterName=S5_RimNoise_Fresnel_Min}
    State_6=ScalarParameter() {group=Rim, parameterName=S6_RimNoise_Fresnel_Min}
   ) {materialFunction=&MF_Radish_StateChangeLerp}.Result) {name=RimNoise_Fresnel_Min, nodeColor=<rgba 1 0.3958 0 1>}
  NamedRerouteDeclaration(None=MaterialFunctionCall(
    State_0=ScalarParameter() {defaultValue=1, group=Rim, parameterName=S0_RimNoise_Fresnel_Max}
    State_1=ScalarParameter() {defaultValue=1, group=Rim, parameterName=S1_RimNoise_Fresnel_Max}
    State_2=ScalarParameter() {defaultValue=1, group=Rim, parameterName=S2_RimNoise_Fresnel_Max}
    State_3=ScalarParameter() {defaultValue=1, group=Rim, parameterName=S3_RimNoise_Fresnel_Max}
    State_4=ScalarParameter() {defaultValue=1, group=Rim, parameterName=S4_RimNoise_Fresnel_Max}
    State_5=ScalarParameter() {defaultValue=1, group=Rim, parameterName=S5_RimNoise_Fresnel_Max}
    State_6=ScalarParameter() {defaultValue=1, group=Rim, parameterName=S6_RimNoise_Fresnel_Max}
   ) {materialFunction=&MF_Radish_StateChangeLerp}.Result) {name=RimNoise_Fresnel_Max, nodeColor=<rgba 1 0.3958 0 1>}
  NamedRerouteDeclaration(None=MaterialFunctionCall(
    State_0=ScalarParameter() {defaultValue=0.2, group=PersistentPulse, parameterName=S0_PersistentPulse_Amount}
    State_1=ScalarParameter() {defaultValue=0.2, group=PersistentPulse, parameterName=S1_PersistentPulse_Amount}
    State_2=ScalarParameter() {defaultValue=0.4, group=PersistentPulse, parameterName=S2_PersistentPulse_Amount}
    State_3=ScalarParameter() {defaultValue=1, group=PersistentPulse, parameterName=S3_PersistentPulse_Amount}
    State_4=ScalarParameter() {defaultValue=4, group=PersistentPulse, parameterName=S4_PersistentPulse_Amount}
    State_5=ScalarParameter() {defaultValue=30, group=PersistentPulse, parameterName=S5_PersistentPulse_Amount}
    State_6=ScalarParameter() {defaultValue=50, group=PersistentPulse, parameterName=S6_PersistentPulse_Amount}
   ) {materialFunction=&MF_Radish_StateChangeLerp}.Result) {name=PersistentPulse_Amount, nodeColor=<rgba 0.7109 0 1 1>}
  NamedRerouteDeclaration(None=MaterialFunctionCall(
    State_0=ScalarParameter() {defaultValue=3, group=Deep Noise, parameterName=S0_Deep_Fresnel_Power}
    State_1=ScalarParameter() {defaultValue=3, group=Deep Noise, parameterName=S1_Deep_Fresnel_Power}
    State_2=ScalarParameter() {defaultValue=3, group=Deep Noise, parameterName=S2_Deep_Fresnel_Power}
    State_3=ScalarParameter() {defaultValue=3, group=Deep Noise, parameterName=S3_Deep_Fresnel_Power}
    State_4=ScalarParameter() {defaultValue=3, group=Deep Noise, parameterName=S4_Deep_Fresnel_Power}
    State_5=ScalarParameter() {defaultValue=3, group=Deep Noise, parameterName=S5_Deep_Fresnel_Power}
    State_6=ScalarParameter() {defaultValue=3, group=Deep Noise, parameterName=S6_Deep_Fresnel_Power}
   ) {materialFunction=&MF_Radish_StateChangeLerp}.Result) {name=Deep_Fresnel_Power, nodeColor=<rgba 1 0.3958 0 1>}
  NamedRerouteDeclaration(None=MaterialFunctionCall(
    State_0=ScalarParameter() {defaultValue=1, group=Deep Noise, parameterName=S0_Deep_Fresnel_Mult}
    State_1=ScalarParameter() {defaultValue=1, group=Deep Noise, parameterName=S1_Deep_Fresnel_Mult}
    State_2=ScalarParameter() {defaultValue=1, group=Deep Noise, parameterName=S2_Deep_Fresnel_Mult}
    State_3=ScalarParameter() {defaultValue=1, group=Deep Noise, parameterName=S3_Deep_Fresnel_Mult}
    State_4=ScalarParameter() {defaultValue=1, group=Deep Noise, parameterName=S4_Deep_Fresnel_Mult}
    State_5=ScalarParameter() {defaultValue=1, group=Deep Noise, parameterName=S5_Deep_Fresnel_Mult}
    State_6=ScalarParameter() {defaultValue=1, group=Deep Noise, parameterName=S6_Deep_Fresnel_Mult}
   ) {materialFunction=&MF_Radish_StateChangeLerp}.Result) {name=Deep_Fresnel_Mult, nodeColor=<rgba 1 0.3958 0 1>}
  NamedRerouteDeclaration(None=MaterialFunctionCall(
    State_0=ScalarParameter() {defaultValue=1, group=Deep Noise, parameterName=S0_Deep_Poles_Tile_X}
    State_1=ScalarParameter() {defaultValue=1, group=Deep Noise, parameterName=S1_Deep_Poles_Tile_X}
    State_2=ScalarParameter() {defaultValue=1, group=Deep Noise, parameterName=S2_Deep_Poles_Tile_X}
    State_3=ScalarParameter() {defaultValue=1, group=Deep Noise, parameterName=S3_Deep_Poles_Tile_X}
    State_4=ScalarParameter() {defaultValue=1, group=Deep Noise, parameterName=S4_Deep_Poles_Tile_X}
    State_5=ScalarParameter() {defaultValue=1, group=Deep Noise, parameterName=S5_Deep_Poles_Tile_X}
    State_6=ScalarParameter() {defaultValue=1, group=Deep Noise, parameterName=S6_Deep_Poles_Tile_X}
   ) {materialFunction=&MF_Radish_StateChangeLerp}.Result) {name=Deep_Poles_Tile_X, nodeColor=<rgba 1 0.4998 0 1>}
  NamedRerouteUsage()
  ScalarParameter() {defaultValue=0.15, parameterName=Spec}
  ScalarParameter() {defaultValue=0.4, parameterName=Roughness}
 ]
}
